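# Material function for the Fearless Flight emote prop that adds switchable anime-style highlights: screen-projected rim light and dot hatching.
# A screen-projected vector dotted with the normal forms a hard-edged rim, adjusted by a pre-skinned-normal Y mask.
# Actor-anchored virtual-plane UVs, scaled by FOV, drive the dots, alongside suit-mask, Lambert and stepped camera-distance terms.
# The output flattens specular and metallic, sets full roughness and adds LUT-processed emissive.
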MaterialFunction /BRCosmetics/Animation/Props/FearlessFlight/Materials/MFC_Gadget_FearlessFlight_Highlights
graph {nodes [
  AtmosphericLightVector()
  FunctionOutput(None=StaticSwitchParameter(True=NamedRerouteUsage(), False=NamedRerouteUsage()) {group=01_FearlessFlight, parameterName=UseFearlessFlightHighlights, sortPriority=0})
  n1 = VectorParameter() {group=01_FearlessFlight, parameterName=RimLighVector, sortPriority=10, defaultValue=<rgba -0.5 0.25 1 1>}
  n2 = NamedRerouteDeclaration(None=DotProduct(
    A=MaterialFunctionCall!(Vector=@n1.RGB) {materialFunction=/Game/Packages/Fortress_SharedMaterials/Anime_Master_Material/MF_ProjectToScreenSpace}.Result
    B=PixelNormalWS()
   )) {name=RimGradient, nodeColor=<rgba 0 0.3906 1 1>}
  n3 = Add(
   A=@n2
   B=NamedRerouteDeclaration(None=ScalarParameter() {group=01_FearlessFlight, parameterName=YMask_Offset, sliderMax=1, sliderMin=-1}) {name=YmaskOffset, nodeColor=<rgba 0 1 0.2434 1>}
  ) {constB=-0.25}
  n4 = DotProduct(A=VertexInterpolator(VS=PreSkinnedNormal()).PS, B=Constant3Vector() {constant=<rgba 0 1 0 0>})
  n5 = Multiply(A=@n4, B=@n4)
  n6 = NamedRerouteDeclaration(None=Multiply(A=@n5, B=@n5) {desc=reduce rim light on the Y axis}) {name=YMask, nodeColor=<rgba 0 1 0.5078 1>}
  n7 = Subtract(
   A=Multiply(
    A=LinearInterpolate(A=@n2, B=@n3, Alpha=@n6) {constA=1, constB=16}
    B=ScalarParameter() {defaultValue=2, group=01_FearlessFlight, parameterName=RimLightGradientScale, sliderMax=4}
   ) {constB=2}
   B=NamedRerouteUsage()
  )
  n8 = NamedRerouteDeclaration(None=SmoothStep(Min=Constant(), Max=Constant() {r=0.005}, Value=Saturate(None=@n7))) {name=AccentRimMask, nodeColor=<rgba 0.3248 0.414 0.9131 1>}
  n9 = VectorParameter() {group=01_FearlessFlight, parameterName=RimColor, sortPriority=10, defaultValue=<rgba 0.9 0.9 0.9 1>}
  NamedRerouteDeclaration(None=AppendVector(A=Multiply(A=@n8, B=@n9.RGB), B=@n9.A)) {name=AccentRim, nodeColor=<rgba 0 0.0489 1 1>}
  n10 = TextureSample() {samplerType=SAMPLERTYPE_Masks, texture=/Game/Global/Textures/Default/Blanks/AllRed}
  NamedRerouteDeclaration(None=ChannelMaskParameter(None=FunctionInput(Preview=@n10.RGBA) {inputName=FX, inputType=FunctionInput_Vector4, sortPriority=1}) {group=01_FearlessFlight, parameterName=SuitBlackMaskChannel, sortPriority=3}.R) {name=SuitMask, nodeColor=<rgba 0.0156 0.0156 0.0156 1>}
  n11 = MaterialFunctionCall() {materialFunction=&CameraDirectionVector}
  n12 = Multiply(A=Constant3Vector() {constant=<rgba 1 1 0 0>}, B=@n11.Result)
  n13 = Add(A=Multiply(B=Normalize(VectorInput=@n12)) {constA=0.25}, B=@n11.Result)
  n14 = Multiply(A=@n11.Result, B=ScalarParameter() {defaultValue=5, group=01_FearlessFlight, parameterName=ScreenPostionDepthBias, sliderMax=100}) {constB=50}
  n15 = NamedRerouteDeclaration(None=Add(A=ActorPositionWS(), B=VectorParameter() {group=01_FearlessFlight, parameterName=VirtualPlanePivotOffset, defaultValue=<rgba 0 0 40 1>}.RGB)) {name=AdjustedActorPosition, nodeColor=<rgba 0.8281 0 1 1>}
  n16 = Constant() {r=64}
  n17 = Constant() {r=128}
  n18 = InverseLinearInterpolate(
   A=Constant() {r=45}
   B=Constant() {r=90}
   Value=MaterialFunctionCall() {materialFunction=/Engine/Functions/Engine_MaterialFunctions02/Utility/FOV}.FOV angle in degrees
  )
  NamedRerouteDeclaration(None=MaterialFunctionCall(Plane Normal Axis=Normalize(VectorInput=@n13), Plane Center=Add(A=@n14, B=@n15), UV Scale=LinearInterpolate(A=@n16, B=@n17, Alpha=@n18)) {materialFunction=&VirtualPlaneCoordinates}.UVs) {name=ObjectAlighedScreenPosition, nodeColor=<rgba 1 0.867 0 1>}
  n19 = Custom() {code="return View.DirectionalLightDirection;", description=DirectionalLightDirection, inputs []}
  n20 = CollectionParameter() {
   collection=&FortniteMaterialParameters
   parameterName=SunlightDirectionalVector
  }
  n21 = ComponentMask(None=@n20) {b=true, g=true, r=true}
  n22 = If(A=DotProduct(A=@n19, B=@n19), "A > B"=@n19, "A == B"=@n21, "A < B"=@n21)
  n23 = PixelNormalWS()
  n24 = Multiply(A=NamedRerouteDeclaration(None=DotProduct(A=@n22, B=@n23)) {name=LdotN, nodeColor=<rgba 0.2656 0 1 1>}) {constB=0.5}
  NamedRerouteDeclaration(None=Saturate(None=Add(A=@n24) {constB=0.5})) {name=Lambert, nodeColor=<rgba 0 0.9297 1 1>}
  n25 = Multiply(
   A=NamedRerouteUsage()
   B=Divide(
    A=ScalarParameter() {defaultValue=150, group=01_FearlessFlight, parameterName=DotsAmount, sliderMax=512, sortPriority=20}
    B=NamedRerouteUsage()
   )
  )
  n26 = Constant() {r=1}
  n27 = Add(A=Frac(None=Rotator(Coordinate=@n25, Time=@n26) {speed=0.75})) {constB=-0.5}
  n28 = Constant()
  NamedRerouteDeclaration(None=Add(A=Distance(A=@n27, B=@n28)) {constB=-0.1, desc=to get full transition}) {name=DotsGradient, nodeColor=<rgba 1 0.0938 0 1>}
  n29 = TextureSample() {samplerType=SAMPLERTYPE_Masks, texture=/BRCosmetics/Animation/Props/FearlessFlight/Textures/T_Emote_Prop_FearlessFlight_S}
  n30 = SetMaterialAttributes!(
   MaterialAttributes
   BaseColor=TextureSample() {texture=/BRCosmetics/Animation/Props/FearlessFlight/Textures/T_Emote_Prop_FearlessFlight_D}.RGB
   Specular=@n29.R
   Metallic=@n29.R
   Roughness=@n29.R
   Normal=TextureSample() {samplerType=SAMPLERTYPE_Normal, texture=/BRCosmetics/Animation/Props/FearlessFlight/Textures/T_Emote_Prop_FearlessFlight_N}.RGB
  ) {attributeSetTypes [
    #1
    #2
    #3
    #4
    #5
   ]}
  NamedRerouteDeclaration(None=FunctionInput(Preview=@n30) {inputName=MA, inputType=FunctionInput_MaterialAttributes}) {name=MA, nodeColor=<rgba 0.3594 1 0 1>}
  n31 = GetMaterialAttributes(None=NamedRerouteUsage()) {attributeGetTypes [#6, #1]}
  n32 = GetMaterialAttributes(None=NamedRerouteUsage()) {attributeGetTypes [#1]}
  n33 = ScalarParameter() {defaultValue=1, group=01_FearlessFlight, parameterName=BCEmissive, sliderMax=1}
  n34 = Multiply(A=NamedRerouteUsage(), B=NamedRerouteUsage())
  n35 = Multiply(A=ComponentMask(None=@n34) {b=true, g=true, r=true}, B=Multiply(A=ComponentMask(None=@n34) {a=true}, B=NamedRerouteUsage())) {desc=Left RIm}
  n36 = Saturate(None=Add(A=Multiply(A=@n32.BaseColor, B=@n33) {constB=0.5}, B=@n35))
  n37 = Multiply(A=NamedRerouteUsage(), B=NamedRerouteUsage())
  n38 = SetMaterialAttributes!(
   MaterialAttributes=@n31.MaterialAttributes
   EmissiveColor=Add(
    A=@n31.EmissiveColor
    B=MaterialFunctionCall!(In=@n36, LUT, LUT Multiplier) {materialFunction=/Game/Packages/Fortress_SharedMaterials/Anime_Master_Material/MF_ApplyLUT}.Result
   )
   BaseColor=LinearInterpolate(B=@n35, Alpha=Saturate(None=@n37))
   Metallic=Constant()
   Specular=Constant()
   Roughness=Constant() {r=1}
  ) {
   desc=screen projected rim and hatching
   attributeSetTypes [
    #6
    #1
    #3
    #2
    #4
   ]
  }
  NamedRerouteDeclaration(None=NamedRerouteDeclaration(None=@n38) {name=AccentRimMaterial, nodeColor=<rgba 0 0.12 1 1>}) {name=MA_Out, nodeColor=<rgba 1 0.0703 0 1>}
  n39 = Subtract(A=Distance(A=NamedRerouteUsage(), B=CameraPositionWS())) {constB=100}
  n40 = ScalarParameter() {defaultValue=6, group=01_FearlessFlight, parameterName=DistanceBlendingSteps, sliderMax=8, sliderMin=1, sortPriority=20}
  n41 = Multiply(A=Saturate(None=Divide(A=@n39) {constB=1200}), B=@n40) {constB=4}
  NamedRerouteDeclaration(None=Max(A=Ceil(None=@n41))) {name=SteppedDistanceBlending, nodeColor=<rgba 0.4766 1 0 1>}
  SkyAtmosphereLightDirection()
 ]}
