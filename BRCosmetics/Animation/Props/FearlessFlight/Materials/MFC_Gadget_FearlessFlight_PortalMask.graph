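# Material function for the FearlessFlight gadget cosmetic that clips a mesh against a portal plane through its opacity mask.
# World position relative to PortalPosition is inverse-transformed by the PortalAxisX/Y/Z vector parameters, and a Step on the local Y component yields a hard half-space mask.
# The UseFearlessFlightPortalMask static switch selects between material attributes with overridden OpacityMask and a branch without it.
# A pre-skinned local position input is declared but does not feed the mask math.
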MaterialFunction /BRCosmetics/Animation/Props/FearlessFlight/Materials/MFC_Gadget_FearlessFlight_PortalMask
graph {nodes [
  n1 = NamedRerouteUsage()
  n2 = NamedRerouteUsage()
  FunctionOutput(None=StaticSwitchParameter(
    True=SetMaterialAttributes!(MaterialAttributes=@n1, OpacityMask=@n2) {attributeSetTypes [#1]}
    False=NamedRerouteUsage()
   ) {group=01_FearlessFlight, parameterName=UseFearlessFlightPortalMask})
  GetMaterialAttributes(None=NamedRerouteDeclaration(None=FunctionInput() {inputName=MA, inputType=FunctionInput_MaterialAttributes}) {name=MA, nodeColor=<rgba 0 0.6484 1 1>}) {attributeGetTypes [#2, #3]}
  n3 = PreSkinnedPosition()
  n4 = VertexInterpolator(VS=@n3)
  NamedRerouteDeclaration(None=FunctionInput(Preview=ShaderStageSwitch(PixelShader=@n4.PS, VertexShader=@n3)) {inputName=PreSkinnnedLocalPosition, sortPriority=3}) {name=PreSkinnedLocalPosition, nodeColor=<rgba 0 1 0.7422 1>}
  n5 = MaterialFunctionCall!(
   VectorToTransform=Subtract(A=WorldPosition().XYZ, B=VectorParameter() {parameterName=PortalPosition, defaultValue=<rgba 0 0 0 1>}.RGB)
   BasisX=VectorParameter() {parameterName=PortalAxisX, defaultValue=<rgba 1 0 0 1>}.RGB
   BasisY=VectorParameter() {parameterName=PortalAxisY, defaultValue=<rgba 0 1 0 1>}.RGB
   BasisZ=VectorParameter() {parameterName=PortalAxisZ, defaultValue=<rgba 0 0 1 1>}.RGB
  ) {materialFunction=/Engine/Functions/Engine_MaterialFunctions02/Math/InverseTransformMatrix}
  NamedRerouteDeclaration(None=Step(X=ComponentMask(None=@n5.Transformed Vector 3) {g=true})) {name=PortalMask, nodeColor=<rgba 0 0.9062 1 1>}
 ]}
