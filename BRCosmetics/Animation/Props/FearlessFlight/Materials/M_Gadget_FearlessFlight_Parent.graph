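# Parent material for the FearlessFlight gadget prop, built on the shared character material functions for skeletal meshes with Default Lit and Subsurface Profile shading.
# Material attributes from the character PreFX function pass through a gadget highlights function driven by the SkinFX_Mask texture parameter.
# A portal mask function then uses pre-skinned local position before the character PostFX and cinematic CharPostFX functions finish the chain.
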Material /BRCosmetics/Animation/Props/FearlessFlight/Materials/M_Gadget_FearlessFlight_Parent
graph {
 properties {
  bAutomaticallySetUsageInEditor=false
  bUseMaterialAttributes=true
  bUsedWithMeshDeformer=true
  bUsedWithMorphTargets=true
  bUsedWithSkeletalMesh=true
  pixelDepthOffsetMode=PDOM_Legacy
  shadingModel=MSM_FromMaterialExpression
  subsurfaceProfile=&SS_HeroSkin_02
  usedShadingModels=Default Lit | Subsurface Profile
 }
 nodes [
  n1 = MaterialFunctionCall() {materialFunction=&BaseFN_CharacterFunction_PreFX}
  n2 = TextureSampleParameter2D() {
   group=Skin FX
   parameterName=SkinFX_Mask
   samplerType=SAMPLERTYPE_Masks
   sortPriority=1
   texture=&T_BlackMask
  }
  n3 = MaterialFunctionCall(
   AO=@n1.AO
   Distance Blend=@n1.AO
   Fresnel=@n1.AO
   Gmap=@n1.AO
   Pre Skinned Local Position=@n1.AO
   Material=MaterialFunctionCall!(
    MA=MaterialFunctionCall!(MA=@n1.Result, FX=@n2.RGBA) {materialFunction=&/MFC_Gadget_FearlessFlight_Highlights}.Result
    PreSkinnnedLocalPosition=@n1.Pre skinned Local Position
   ) {materialFunction=&/MFC_Gadget_FearlessFlight_PortalMask}.Result
  ) {materialFunction=&BaseFN_CharacterFunction_PostFX}
  MaterialFunctionCall(In=@n3.Result) {materialFunction=&CINE_MF_CharPostFX}
 ]
}
